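# Physics asset for the back quills of the MixedAlley male medium head's emote accessory, using capsule bodies throughout.
# Five quill chains (back and forward pairs plus a center one) each run from a small base on spine_05 through three dynamic segments.
# Segment joints lock twist and limit swing, with one swing axis much wider beyond the first.
# Spine, clavicle and upper-arm capsules approximate the torso, linked by free-rotation constraints.
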PhysicsAsset /BRCosmetics/Characters/Player/Male/Medium/Heads/M_MED_MixedAlley_Head/Meshes/CustomSkeleton/M_MED_MixedAlley_FaceAcc_Back_Emote_Physics
bodies {
 base_back_bk_quill_l [{shapeType=Capsule, length=0.505, radius=0.505, shapeName=base_back_bk_quill_l_capsule, rotation=<rot 0 -90 0>}]
 base_back_bk_quill_r [{shapeType=Capsule, length=0.505, radius=0.505, shapeName=base_back_bk_quill_r_capsule, rotation=<rot 0 -90 0>}]
 base_back_fw_quill_l [{shapeType=Capsule, length=0.505, radius=0.505, shapeName=base_back_fw_quill_l_capsule, rotation=<rot 0 -90 0>}]
 base_back_fw_quill_r [{shapeType=Capsule, length=0.505, radius=0.505, shapeName=base_back_fw_quill_r_capsule, rotation=<rot 0 -90 0>}]
 base_back_quill_c [{shapeType=Capsule, length=0.505, radius=0.505, shapeName=base_back_quill_c_capsule, rotation=<rot 0 -90 0>}]
 clavicle_l [{
   shapeType=Capsule
   length=4.6779
   radius=6.1717
   shapeName=clavicle_l_capsule
   center=<6.3931 -1.0251 -2.2485>
   rotation=<rot 86.9643 -0.0014 -0.0017>
  }]
 clavicle_r [{
   shapeType=Capsule
   length=4.6779
   radius=6.1717
   shapeName=clavicle_l_capsule
   center=<-6.3931 1.0251 2.2485>
   rotation=<rot 86.9643 -0.0014 179.9983>
  }]
 dyn_back_bk_quill_01_l [{
   shapeType=Capsule
   length=14.6798
   radius=7.8582
   shapeName=dyn_back_bk_quill_01_l_capsule
   center=<9.6966 0.4814 -0.7542>
   rotation=<rot 0 -85.8319 -81.3638>
  }]
 dyn_back_bk_quill_01_r [{
   shapeType=Capsule
   length=14.6798
   radius=7.8581
   shapeName=dyn_back_bk_quill_01_r_capsule
   center=<9.697 -0.4814 -0.7542>
   rotation=<rot 0 -94.1675 -81.3634>
  }]
 dyn_back_bk_quill_02_l [{
   shapeType=Capsule
   length=8.1858
   radius=7.2405
   shapeName=dyn_back_bk_quill_02_l_capsule
   center=<6.789 0.0458 -0.9303>
   rotation=<rot 2.7935 -89.6958 -83.7155>
  }]
 dyn_back_bk_quill_02_r [{
   shapeType=Capsule
   length=8.1862
   radius=7.241
   shapeName=dyn_back_bk_quill_02_r_capsule
   center=<7.3158 -0.0486 -0.9884>
   rotation=<rot -2.1784 -95.7914 -83.4766>
  }]
 dyn_back_bk_quill_03_l [{
   shapeType=Capsule
   length=5.6833
   radius=5.3419
   shapeName=dyn_back_bk_quill_03_l_capsule
   center=<7.1443 -0.0112 -0.7257>
   rotation=<rot 0 91.8442 -89.159>
  }]
 dyn_back_bk_quill_03_r [{
   shapeType=Capsule
   length=5.6833
   radius=5.3419
   shapeName=dyn_back_bk_quill_03_r_capsule
   center=<3.219 0.1376 -0.7834>
   rotation=<rot 0 88.1558 -89.159>
  }]
 dyn_back_fw_quill_01_l [{
   shapeType=Capsule
   length=10.6371
   radius=6.555
   shapeName=dyn_back_fw_quill_01_l_capsule
   center=<11.9543 -0.3745 -0.5191>
   rotation=<rot 0 -93.195 -86.3634>
  }]
 dyn_back_fw_quill_01_r [{
   shapeType=Capsule
   length=10.6388
   radius=6.5534
   shapeName=dyn_back_fw_quill_01_r_capsule
   center=<11.9552 -0.3697 0.5168>
   rotation=<rot 0 86.8571 -86.4053>
  }]
 dyn_back_fw_quill_02_l [{
   shapeType=Capsule
   length=8.0115
   radius=4.9346
   shapeName=dyn_back_fw_quill_02_l_capsule
   center=<6.3456 -0.7487 -0.1652>
   rotation=<rot 0 -95.8612 -85.5478>
  }]
 dyn_back_fw_quill_02_r [{
   shapeType=Capsule
   length=8.0115
   radius=4.9346
   shapeName=dyn_back_fw_quill_02_r_capsule
   center=<5.7637 -0.689 0.1197>
   rotation=<rot 0 84.1388 -85.5478>
  }]
 dyn_back_fw_quill_03_l [{
   shapeType=Capsule
   length=8.9646
   radius=4.0421
   shapeName=dyn_back_fw_quill_03_l_capsule
   center=<4.4699 -0.4212 -0.5647>
   rotation=<rot 0 88.6937 -89.5648>
  }]
 dyn_back_fw_quill_03_r [{
   shapeType=Capsule
   length=8.9646
   radius=4.0421
   shapeName=dyn_back_fw_quill_03_r_capsule
   center=<5.5406 -0.4456 -0.0249>
   rotation=<rot 0 -91.3063 -89.5648>
  }]
 dyn_back_quill_01_c [{
   shapeType=Capsule
   length=9.4036
   radius=6.7981
   shapeName=dyn_back_quill_01_c_capsule
   center=<6.8941 -0.1328 -1.0481>
   rotation=<rot 0.1314 -89.0314 -89.9998>
  }]
 dyn_back_quill_02_c [{
   shapeType=Capsule
   length=9.6864
   radius=8.9321
   shapeName=dyn_back_quill_02_c_capsule
   center=<4.494 -0.0247 0.2933>
   rotation=<rot 0 90.2118 -84.8514>
  }]
 dyn_back_quill_03_c [{
   shapeType=Capsule
   length=7.12
   radius=6.0596
   shapeName=dyn_back_quill_03_c_capsule
   center=<8.4288 -0.0335 -1.4187>
   rotation=<rot 0 -90.1179 -81.1847>
  }]
 spine_03 [{shapeType=Capsule, length=4.8535, radius=12.8982, shapeName=spine_03_capsule, center=<-2.0838 2.2176 0>}]
 spine_04 [{shapeType=Capsule, length=8.979, radius=14.3105, shapeName=spine_04_capsule, center=<0 2.6192 0>}]
 spine_05 [{shapeType=Capsule, length=10.2422, radius=12.8459, shapeName=spine_05_capsule, center=<0.0464 -1.6321 0>}]
 upperarm_l [{shapeType=Capsule, length=16.5881, radius=2.755, shapeName=upperarm_l_capsule, center=<10.4447 0 0>, rotation=<rot -0.0001 -89.9999 -90>}]
 upperarm_r [{shapeType=Capsule, length=16.5881, radius=2.755, shapeName=upperarm_r_capsule, center=<-10.4447 0 0>, rotation=<rot -0.0001 -89.9999 90>}]
}
constraints [
 {
  bone1Name=spine_05
  bone2Name=spine_04
  swing1LimitDegrees=45
  swing1Motion=Free
  swing2LimitDegrees=45
  swing2Motion=Free
  twistLimitDegrees=45
  twistMotion=Free
 }
 {
  bone1Name=clavicle_l
  bone2Name=spine_05
  swing1LimitDegrees=45
  swing1Motion=Free
  swing2LimitDegrees=45
  swing2Motion=Free
  twistLimitDegrees=45
  twistMotion=Free
 }
 {
  bone1Name=clavicle_r
  bone2Name=spine_05
  swing1LimitDegrees=45
  swing1Motion=Free
  swing2LimitDegrees=45
  swing2Motion=Free
  twistLimitDegrees=45
  twistMotion=Free
 }
 {
  bone1Name=upperarm_l
  bone2Name=clavicle_l
  swing1LimitDegrees=45
  swing1Motion=Free
  swing2LimitDegrees=45
  swing2Motion=Free
  twistLimitDegrees=45
  twistMotion=Free
 }
 {
  bone1Name=upperarm_r
  bone2Name=clavicle_r
  swing1LimitDegrees=45
  swing1Motion=Free
  swing2LimitDegrees=45
  swing2Motion=Free
  twistLimitDegrees=45
  twistMotion=Free
 }
 {
  bone1Name=spine_04
  bone2Name=spine_03
  swing1LimitDegrees=45
  swing1Motion=Free
  swing2LimitDegrees=45
  swing2Motion=Free
  twistLimitDegrees=45
  twistMotion=Free
 }
 {
  bone1Name=base_back_bk_quill_l
  bone2Name=spine_05
  swing1LimitDegrees=45
  swing1Motion=Free
  swing2LimitDegrees=45
  swing2Motion=Free
  twistLimitDegrees=45
  twistMotion=Free
 }
 {
  bone1Name=dyn_back_bk_quill_01_l
  bone2Name=base_back_bk_quill_l
  swing1LimitDegrees=8
  swing2LimitDegrees=10
  twistLimitDegrees=1
  twistMotion=Locked
 }
 {
  bone1Name=dyn_back_bk_quill_02_l
  bone2Name=dyn_back_bk_quill_01_l
  swing1LimitDegrees=10
  swing2LimitDegrees=60
  twistLimitDegrees=1
  twistMotion=Locked
 }
 {
  bone1Name=dyn_back_bk_quill_03_l
  bone2Name=dyn_back_bk_quill_02_l
  swing1LimitDegrees=8
  swing2LimitDegrees=60
  twistLimitDegrees=1
  twistMotion=Locked
 }
 {
  bone1Name=base_back_bk_quill_r
  bone2Name=spine_05
  swing1LimitDegrees=45
  swing1Motion=Free
  swing2LimitDegrees=45
  swing2Motion=Free
  twistLimitDegrees=45
  twistMotion=Free
 }
 {
  bone1Name=dyn_back_bk_quill_01_r
  bone2Name=base_back_bk_quill_r
  swing1LimitDegrees=8
  swing2LimitDegrees=10
  twistLimitDegrees=1
  twistMotion=Locked
 }
 {
  bone1Name=dyn_back_bk_quill_02_r
  bone2Name=dyn_back_bk_quill_01_r
  swing1LimitDegrees=10
  swing2LimitDegrees=60
  twistLimitDegrees=1
  twistMotion=Locked
 }
 {
  bone1Name=dyn_back_bk_quill_03_r
  bone2Name=dyn_back_bk_quill_02_r
  swing1LimitDegrees=8
  swing2LimitDegrees=60
  twistLimitDegrees=1
  twistMotion=Locked
 }
 {
  bone1Name=base_back_fw_quill_l
  bone2Name=spine_05
  swing1LimitDegrees=45
  swing1Motion=Free
  swing2LimitDegrees=45
  swing2Motion=Free
  twistLimitDegrees=45
  twistMotion=Free
 }
 {
  bone1Name=dyn_back_fw_quill_01_l
  bone2Name=base_back_fw_quill_l
  swing1LimitDegrees=8
  swing2LimitDegrees=10
  twistLimitDegrees=1
  twistMotion=Locked
 }
 {
  bone1Name=dyn_back_fw_quill_02_l
  bone2Name=dyn_back_fw_quill_01_l
  swing1LimitDegrees=10
  swing2LimitDegrees=60
  twistLimitDegrees=1
  twistMotion=Locked
 }
 {
  bone1Name=dyn_back_fw_quill_03_l
  bone2Name=dyn_back_fw_quill_02_l
  swing1LimitDegrees=8
  swing2LimitDegrees=60
  twistLimitDegrees=1
  twistMotion=Locked
 }
 {
  bone1Name=base_back_fw_quill_r
  bone2Name=spine_05
  swing1LimitDegrees=45
  swing1Motion=Free
  swing2LimitDegrees=45
  swing2Motion=Free
  twistLimitDegrees=45
  twistMotion=Free
 }
 {
  bone1Name=dyn_back_fw_quill_01_r
  bone2Name=base_back_fw_quill_r
  swing1LimitDegrees=8
  swing2LimitDegrees=10
  twistLimitDegrees=1
  twistMotion=Locked
 }
 {
  bone1Name=dyn_back_fw_quill_02_r
  bone2Name=dyn_back_fw_quill_01_r
  swing1LimitDegrees=10
  swing2LimitDegrees=60
  twistLimitDegrees=1
  twistMotion=Locked
 }
 {
  bone1Name=dyn_back_fw_quill_03_r
  bone2Name=dyn_back_fw_quill_02_r
  swing1LimitDegrees=8
  swing2LimitDegrees=60
  twistLimitDegrees=1
  twistMotion=Locked
 }
 {
  bone1Name=base_back_quill_c
  bone2Name=spine_05
  swing1LimitDegrees=45
  swing1Motion=Free
  swing2LimitDegrees=45
  swing2Motion=Free
  twistLimitDegrees=45
  twistMotion=Free
 }
 {
  bone1Name=dyn_back_quill_01_c
  bone2Name=base_back_quill_c
  swing1LimitDegrees=8
  swing2LimitDegrees=10
  twistLimitDegrees=1
  twistMotion=Locked
 }
 {
  bone1Name=dyn_back_quill_02_c
  bone2Name=dyn_back_quill_01_c
  swing1LimitDegrees=10
  swing2LimitDegrees=60
  twistLimitDegrees=1
  twistMotion=Locked
 }
 {
  bone1Name=dyn_back_quill_03_c
  bone2Name=dyn_back_quill_02_c
  swing1LimitDegrees=8
  swing2LimitDegrees=60
  twistLimitDegrees=1
  twistMotion=Locked
 }
]
