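# Physics asset for the back quill accessory of the male medium MixedAlley head cosmetic, defining bodies and joints for five quill chains.
# Each chain (back and forward pairs plus center) links a small base capsule on spine_05 to three dyn capsules.
# Spine, clavicle and upper-arm shapes model the torso, joined to each other and the bases by free-rotation constraints.
# Quill joints lock twist and limit swing, generally allowing more bend toward the tips.
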PhysicsAsset /BRCosmetics/Characters/Player/Male/Medium/Heads/M_MED_MixedAlley_Head/Meshes/CustomSkeleton/M_MED_MixedAlley_FaceAcc_Back_Physics
bodies {
 base_back_bk_quill_l [{shapeType=Capsule, length=0.505, radius=0.505, shapeName=base_back_bk_quill_l_capsule, rotation=<rot 0 -90 0>}]
 base_back_bk_quill_r [{shapeType=Capsule, length=0.505, radius=0.505, shapeName=base_back_bk_quill_r_capsule, rotation=<rot 0 -90 0>}]
 base_back_fw_quill_l [{shapeType=Capsule, length=0.505, radius=0.505, shapeName=base_back_fw_quill_l_capsule, rotation=<rot 0 -90 0>}]
 base_back_fw_quill_r [{shapeType=Capsule, length=0.505, radius=0.505, shapeName=base_back_fw_quill_r_capsule, rotation=<rot 0 -90 0>}]
 base_back_quill_c [{shapeType=Capsule, length=0.505, radius=0.505, shapeName=base_back_quill_c_capsule, rotation=<rot 0 -90 0>}]
 clavicle_l [{
   shapeType=Capsule
   length=9.326
   radius=6.9091
   shapeName=clavicle_l_capsule
   center=<6.4633 -0.7416 -0.4731>
   rotation=<rot 86.9638 -0.1986 -5.0246>
  }]
 clavicle_r [{
   shapeType=Capsule
   length=9.326
   radius=6.8937
   shapeName=clavicle_l_capsule
   center=<-6.4547 0.7416 0.6347>
   rotation=<rot 86.9638 -0.1986 -5.0246>
  }]
 dyn_back_bk_quill_01_l [{
   shapeType=Capsule
   length=14.6798
   radius=6.2935
   shapeName=dyn_back_bk_quill_01_l_capsule
   center=<9.3912 0.0404 0.4159>
   rotation=<rot -0.0884 -89.1792 91.0125>
  }]
 dyn_back_bk_quill_01_r [{
   shapeType=Capsule
   length=11.7575
   radius=6.2935
   shapeName=dyn_back_bk_quill_01_r_capsule
   center=<8.0685 0.2102 -0.7245>
   rotation=<rot -0.0884 -89.1792 -88.9875>
  }]
 dyn_back_bk_quill_02_l [{
   shapeType=Capsule
   length=5.8364
   radius=3.9158
   shapeName=dyn_back_bk_quill_02_l_capsule
   center=<5.431 0.0321 -0.056>
   rotation=<rot 1.4717 -88.5696 89.4085>
  }]
 dyn_back_bk_quill_02_r [{
   shapeType=Capsule
   length=14.1407
   radius=3.9158
   shapeName=dyn_back_bk_quill_02_r_capsule
   center=<10.3586 0.2711 0.2214>
   rotation=<rot 1.4253 -92.6913 -90.6958>
  }]
 dyn_back_bk_quill_03_l [{
   shapeType=Capsule
   length=5.0089
   radius=3.4
   shapeName=dyn_back_bk_quill_03_l_capsule
   center=<4.679 0.1803 0.3231>
   rotation=<rot -0.3343 93.0341 92.3775>
  }]
 dyn_back_bk_quill_03_r [{
   shapeType=Capsule
   length=5.0089
   radius=3.4
   shapeName=dyn_back_bk_quill_03_r_capsule
   center=<4.1325 0.2889 0.0432>
   rotation=<rot -0.3343 93.0341 -87.6225>
  }]
 dyn_back_fw_quill_01_l [{
   shapeType=Capsule
   length=10.6371
   radius=5.1569
   shapeName=dyn_back_fw_quill_01_l_capsule
   center=<8.4246 -0.2201 -0.3467>
   rotation=<rot -0.0395 -91.1038 -88.9193>
  }]
 dyn_back_fw_quill_01_r [{
   shapeType=Capsule
   length=10.6371
   radius=5.1569
   shapeName=dyn_back_fw_quill_01_r_capsule
   center=<8.3404 -0.1027 0.0303>
   rotation=<rot -0.0395 -91.1038 91.0807>
  }]
 dyn_back_fw_quill_02_l [{
   shapeType=Capsule
   length=7.249
   radius=3.8425
   shapeName=dyn_back_fw_quill_02_l_capsule
   center=<6.0251 0.0094 -0.1344>
   rotation=<rot -0.4721 -89.7867 -85.5728>
  }]
 dyn_back_fw_quill_02_r [{
   shapeType=Capsule
   length=7.249
   radius=3.8425
   shapeName=dyn_back_fw_quill_02_r_capsule
   center=<5.8813 0.0425 -0.7874>
   rotation=<rot -0.4721 -89.7867 94.4272>
  }]
 dyn_back_fw_quill_03_l [{
   shapeType=Capsule
   length=5.7135
   radius=2.5603
   shapeName=dyn_back_fw_quill_03_l_capsule
   center=<3.7224 -0.4042 -0.1297>
   rotation=<rot 0 88.6938 -90.0929>
  }]
 dyn_back_fw_quill_03_r [{
   shapeType=Capsule
   length=5.7135
   radius=2.5603
   shapeName=dyn_back_fw_quill_03_r_capsule
   center=<4.2331 0.2228 0.1168>
   rotation=<rot 0 88.6938 89.9071>
  }]
 dyn_back_quill_01_c [{
   shapeType=Capsule
   length=9.4036
   radius=7.271
   shapeName=dyn_back_quill_01_c_capsule
   center=<6.8941 -0.1311 -0.3185>
   rotation=<rot 0.1314 -89.0314 -89.9998>
  }]
 dyn_back_quill_02_c [{
   shapeType=Capsule
   length=5.4483
   radius=6.1644
   shapeName=dyn_back_quill_02_c_capsule
   center=<4.494 -0.0247 0.2933>
   rotation=<rot 0 90.2118 -91.2508>
  }]
 dyn_back_quill_03_c [{
   shapeType=Capsule
   length=4.1227
   radius=4.9942
   shapeName=dyn_back_quill_03_c_capsule
   center=<5.2775 0.1369 -0.5672>
   rotation=<rot 0 -90.1179 -81.1847>
  }]
 spine_03 [{shapeType=Capsule, length=15.0243, radius=14.8962, shapeName=spine_03_capsule, center=<-2.0838 -2.9013 0>}]
 spine_04 [{shapeType=Capsule, length=14.3608, radius=16.3801, shapeName=spine_04_capsule, center=<0 0.2904 0>}]
 spine_05 [{shapeType=Capsule, length=9.1112, radius=15.4887, shapeName=spine_05_capsule, center=<1.8174 -0.5588 0.0146>}]
 upperarm_l [
  {shapeType=Sphere, radius=7.3893, shapeName=upperarm_l_sphere, center=<0 1.05 1.3985>}
  {
   shapeType=Capsule
   length=24.3425
   radius=5.6575
   shapeName=upperarm_l_capsule
   center=<12.6605 -0.1968 -0.0207>
   rotation=<rot -89.9062 179.9462 -179.9461>
  }
 ]
 upperarm_r [
  {shapeType=Sphere, radius=8.1502, shapeName=upperarm_r_sphere, center=<0 -0.6512 0>}
  {
   shapeType=Capsule
   length=24.3425
   radius=5.6575
   shapeName=upperarm_l_capsule
   center=<-12.6605 0.1968 0.0207>
   rotation=<rot -89.9062 179.9462 0.0539>
  }
 ]
}
constraints [
 {
  bone1Name=spine_05
  bone2Name=spine_04
  swing1LimitDegrees=45
  swing1Motion=Free
  swing2LimitDegrees=45
  swing2Motion=Free
  twistLimitDegrees=45
  twistMotion=Free
 }
 {
  bone1Name=clavicle_l
  bone2Name=spine_05
  swing1LimitDegrees=45
  swing1Motion=Free
  swing2LimitDegrees=45
  swing2Motion=Free
  twistLimitDegrees=45
  twistMotion=Free
 }
 {
  bone1Name=clavicle_r
  bone2Name=spine_05
  swing1LimitDegrees=45
  swing1Motion=Free
  swing2LimitDegrees=45
  swing2Motion=Free
  twistLimitDegrees=45
  twistMotion=Free
 }
 {
  bone1Name=upperarm_l
  bone2Name=clavicle_l
  swing1LimitDegrees=45
  swing1Motion=Free
  swing2LimitDegrees=45
  swing2Motion=Free
  twistLimitDegrees=45
  twistMotion=Free
 }
 {
  bone1Name=upperarm_r
  bone2Name=clavicle_r
  swing1LimitDegrees=45
  swing1Motion=Free
  swing2LimitDegrees=45
  swing2Motion=Free
  twistLimitDegrees=45
  twistMotion=Free
 }
 {
  bone1Name=spine_04
  bone2Name=spine_03
  swing1LimitDegrees=45
  swing1Motion=Free
  swing2LimitDegrees=45
  swing2Motion=Free
  twistLimitDegrees=45
  twistMotion=Free
 }
 {
  bone1Name=base_back_bk_quill_l
  bone2Name=spine_05
  swing1LimitDegrees=45
  swing1Motion=Free
  swing2LimitDegrees=45
  swing2Motion=Free
  twistLimitDegrees=45
  twistMotion=Free
 }
 {
  bone1Name=dyn_back_bk_quill_01_l
  bone2Name=base_back_bk_quill_l
  swing1LimitDegrees=15
  swing2LimitDegrees=30
  twistLimitDegrees=1
  twistMotion=Locked
 }
 {
  bone1Name=dyn_back_bk_quill_02_l
  bone2Name=dyn_back_bk_quill_01_l
  swing1LimitDegrees=25
  swing2LimitDegrees=45
  twistLimitDegrees=1
  twistMotion=Locked
 }
 {
  bone1Name=dyn_back_bk_quill_03_l
  bone2Name=dyn_back_bk_quill_02_l
  swing1LimitDegrees=30
  swing2LimitDegrees=65
  twistLimitDegrees=1
  twistMotion=Locked
 }
 {
  bone1Name=base_back_bk_quill_r
  bone2Name=spine_05
  swing1LimitDegrees=45
  swing1Motion=Free
  swing2LimitDegrees=45
  swing2Motion=Free
  twistLimitDegrees=45
  twistMotion=Free
 }
 {
  bone1Name=dyn_back_bk_quill_01_r
  bone2Name=base_back_bk_quill_r
  swing1LimitDegrees=20
  swing2LimitDegrees=25
  twistLimitDegrees=1
  twistMotion=Locked
 }
 {
  bone1Name=dyn_back_bk_quill_02_r
  bone2Name=dyn_back_bk_quill_01_r
  swing1LimitDegrees=15
  swing2LimitDegrees=45
  twistLimitDegrees=1
  twistMotion=Locked
 }
 {
  bone1Name=dyn_back_bk_quill_03_r
  bone2Name=dyn_back_bk_quill_02_r
  swing1LimitDegrees=35
  swing2LimitDegrees=65
  twistLimitDegrees=1
  twistMotion=Locked
 }
 {
  bone1Name=base_back_fw_quill_l
  bone2Name=spine_05
  swing1LimitDegrees=45
  swing1Motion=Free
  swing2LimitDegrees=45
  swing2Motion=Free
  twistLimitDegrees=45
  twistMotion=Free
 }
 {
  bone1Name=dyn_back_fw_quill_01_l
  bone2Name=base_back_fw_quill_l
  swing1LimitDegrees=35
  swing2LimitDegrees=20
  twistLimitDegrees=1
  twistMotion=Locked
 }
 {
  bone1Name=dyn_back_fw_quill_02_l
  bone2Name=dyn_back_fw_quill_01_l
  swing1LimitDegrees=20
  swing2LimitDegrees=55
  twistLimitDegrees=1
  twistMotion=Locked
 }
 {
  bone1Name=dyn_back_fw_quill_03_l
  bone2Name=dyn_back_fw_quill_02_l
  swing1LimitDegrees=65
  swing2LimitDegrees=25
  twistLimitDegrees=1
  twistMotion=Locked
 }
 {
  bone1Name=base_back_fw_quill_r
  bone2Name=spine_05
  swing1LimitDegrees=45
  swing1Motion=Free
  swing2LimitDegrees=45
  swing2Motion=Free
  twistLimitDegrees=45
  twistMotion=Free
 }
 {
  bone1Name=dyn_back_fw_quill_01_r
  bone2Name=base_back_fw_quill_r
  swing1LimitDegrees=30
  swing2LimitDegrees=45
  twistLimitDegrees=1
  twistMotion=Locked
 }
 {
  bone1Name=dyn_back_fw_quill_02_r
  bone2Name=dyn_back_fw_quill_01_r
  swing1LimitDegrees=30
  swing2LimitDegrees=50
  twistLimitDegrees=1
  twistMotion=Locked
 }
 {
  bone1Name=dyn_back_fw_quill_03_r
  bone2Name=dyn_back_fw_quill_02_r
  swing1LimitDegrees=45
  swing2LimitDegrees=60
  twistLimitDegrees=1
  twistMotion=Locked
 }
 {
  bone1Name=base_back_quill_c
  bone2Name=spine_05
  swing1LimitDegrees=45
  swing1Motion=Free
  swing2LimitDegrees=45
  swing2Motion=Free
  twistLimitDegrees=45
  twistMotion=Free
 }
 {
  bone1Name=dyn_back_quill_01_c
  bone2Name=base_back_quill_c
  swing1LimitDegrees=159
  swing2LimitDegrees=45
  twistLimitDegrees=1
  twistMotion=Locked
 }
 {
  bone1Name=dyn_back_quill_02_c
  bone2Name=dyn_back_quill_01_c
  swing1LimitDegrees=30
  swing2LimitDegrees=60
  twistLimitDegrees=1
  twistMotion=Locked
 }
 {
  bone1Name=dyn_back_quill_03_c
  bone2Name=dyn_back_quill_02_c
  swing1LimitDegrees=45
  swing2LimitDegrees=75
  twistLimitDegrees=1
  twistMotion=Locked
 }
]
