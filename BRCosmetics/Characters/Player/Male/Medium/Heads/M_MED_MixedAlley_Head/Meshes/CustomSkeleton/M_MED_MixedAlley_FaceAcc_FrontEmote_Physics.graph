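# Physics asset for the MixedAlley male head's FrontEmote face accessory, covering a left and a right front quill.
# Each quill is a three-segment capsule chain on a small base body, with very tight twist limits and one swing limit narrowing toward the tip.
# Capsules on the upper spine, clavicles and upper arms approximate the torso.
# These bodies and the quill bases link through constraints with free angular motion.
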PhysicsAsset /BRCosmetics/Characters/Player/Male/Medium/Heads/M_MED_MixedAlley_Head/Meshes/CustomSkeleton/M_MED_MixedAlley_FaceAcc_FrontEmote_Physics
bodies {
 base_front_quill_l [{shapeType=Capsule, length=0.505, radius=0.505, rotation=<rot 0 -90 0>}]
 base_front_quill_r [{shapeType=Capsule, length=0.505, radius=0.505, rotation=<rot 0 -90 0>}]
 clavicle_l [{
   shapeType=Capsule
   length=4.6779
   radius=6.1717
   shapeName=clavicle_l_capsule
   center=<6.3931 -1.0251 -2.2485>
   rotation=<rot 86.9643 -0.0014 -0.0017>
  }]
 clavicle_r [{
   shapeType=Capsule
   length=4.6779
   radius=6.1717
   shapeName=clavicle_l_capsule
   center=<-6.3931 1.0251 2.2485>
   rotation=<rot 86.9643 -0.0014 179.9983>
  }]
 dyn_front_quill_01_l [{shapeType=Capsule, length=9.291, radius=5.7014, center=<8.376 -0.525 0.2146>, rotation=<rot 0 -95.9006 -80.5973>}]
 dyn_front_quill_01_r [{shapeType=Capsule, length=9.291, radius=5.7014, center=<8.3763 -0.5251 -0.2146>, rotation=<rot 0 84.0994 -80.5973>}]
 dyn_front_quill_02_l [{shapeType=Capsule, length=7.4661, radius=4.5016, center=<4.1474 -0.239 0.3327>, rotation=<rot 0 -86.4239 -85.8931>}]
 dyn_front_quill_02_r [{shapeType=Capsule, length=7.4661, radius=4.5016, center=<4.1471 -0.239 -0.3326>, rotation=<rot 0 93.5761 -85.8931>}]
 dyn_front_quill_03_l [{shapeType=Capsule, length=8.6675, radius=3.2833, center=<0.0128 -0.0845 -0.3373>, rotation=<rot 0 -90.6009 -77.8845>}]
 dyn_front_quill_03_r [{shapeType=Capsule, length=8.6675, radius=3.2833, center=<0.013 -0.0845 0.3372>, rotation=<rot 0 89.3991 -77.8845>}]
 spine_03 [{shapeType=Capsule, length=4.8535, radius=12.8982, shapeName=spine_03_capsule, center=<-2.0838 2.2176 0>}]
 spine_04 [{shapeType=Capsule, length=8.979, radius=14.3105, shapeName=spine_04_capsule, center=<0 2.6192 0>}]
 spine_05 [{shapeType=Capsule, length=10.2422, radius=12.8459, shapeName=spine_05_capsule, center=<0.0464 -1.6321 0>}]
 upperarm_l [{shapeType=Capsule, length=16.5881, radius=2.755, shapeName=upperarm_l_capsule, center=<10.4447 0 0>, rotation=<rot -0.0001 -89.9999 -90>}]
 upperarm_r [{shapeType=Capsule, length=16.5881, radius=2.755, shapeName=upperarm_r_capsule, center=<-10.4447 0 0>, rotation=<rot -0.0001 -89.9999 90>}]
}
constraints [
 {bone1Name=dyn_front_quill_01_r, bone2Name=base_front_quill_r, swing1LimitDegrees=25, swing2LimitDegrees=60, twistLimitDegrees=1}
 {bone1Name=dyn_front_quill_02_r, bone2Name=dyn_front_quill_01_r, swing1LimitDegrees=25, swing2LimitDegrees=45, twistLimitDegrees=1}
 {bone1Name=dyn_front_quill_03_r, bone2Name=dyn_front_quill_02_r, swing1LimitDegrees=25, swing2LimitDegrees=25, twistLimitDegrees=1}
 {bone1Name=dyn_front_quill_01_l, bone2Name=base_front_quill_l, swing1LimitDegrees=25, swing2LimitDegrees=60, twistLimitDegrees=1}
 {bone1Name=dyn_front_quill_02_l, bone2Name=dyn_front_quill_01_l, swing1LimitDegrees=25, swing2LimitDegrees=45, twistLimitDegrees=1}
 {bone1Name=dyn_front_quill_03_l, bone2Name=dyn_front_quill_02_l, swing1LimitDegrees=25, swing2LimitDegrees=25, twistLimitDegrees=1}
 {
  bone1Name=base_front_quill_r
  bone2Name=spine_05
  swing1LimitDegrees=45
  swing1Motion=Free
  swing2LimitDegrees=45
  swing2Motion=Free
  twistLimitDegrees=45
  twistMotion=Free
 }
 {
  bone1Name=base_front_quill_l
  bone2Name=spine_05
  swing1LimitDegrees=45
  swing1Motion=Free
  swing2LimitDegrees=45
  swing2Motion=Free
  twistLimitDegrees=45
  twistMotion=Free
 }
 {
  bone1Name=spine_05
  bone2Name=spine_04
  swing1LimitDegrees=45
  swing1Motion=Free
  swing2LimitDegrees=45
  swing2Motion=Free
  twistLimitDegrees=45
  twistMotion=Free
 }
 {
  bone1Name=clavicle_l
  bone2Name=spine_05
  swing1LimitDegrees=45
  swing1Motion=Free
  swing2LimitDegrees=45
  swing2Motion=Free
  twistLimitDegrees=45
  twistMotion=Free
 }
 {
  bone1Name=clavicle_r
  bone2Name=spine_05
  swing1LimitDegrees=45
  swing1Motion=Free
  swing2LimitDegrees=45
  swing2Motion=Free
  twistLimitDegrees=45
  twistMotion=Free
 }
 {
  bone1Name=upperarm_r
  bone2Name=clavicle_r
  swing1LimitDegrees=45
  swing1Motion=Free
  swing2LimitDegrees=45
  swing2Motion=Free
  twistLimitDegrees=45
  twistMotion=Free
 }
 {
  bone1Name=spine_04
  bone2Name=spine_03
  swing1LimitDegrees=45
  swing1Motion=Free
  swing2LimitDegrees=45
  swing2Motion=Free
  twistLimitDegrees=45
  twistMotion=Free
 }
 {
  bone1Name=upperarm_l
  bone2Name=clavicle_l
  swing1LimitDegrees=45
  swing1Motion=Free
  swing2LimitDegrees=45
  swing2Motion=Free
  twistLimitDegrees=45
  twistMotion=Free
 }
]
