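# Post-process AnimBP for the MixedAlley male head, driving its quills and adapting them to emotes, aiming and back accessories.
# A Control Rig and head counter-rotation feed cached front-quill, then back-quill RigidBody passes, with physics assets swapped by state.
# Backbling and capes trigger additive bone offsets that tuck or reshape back quills.
# Local-player aiming with most visible weapons shrinks head, neck and a front quill.
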AnimBlueprint /BRCosmetics/Characters/Player/Male/Medium/Heads/M_MED_MixedAlley_Head/Meshes/CustomSkeleton/M_MED_MixedAlley_Head_AnimBP_PP
target_skeleton=&/Fortnite_Base_Head_MixedAlley_Skeleton
graphs [
 AnimGraph {nodes [
   Component To Local(ComponentPose=Local To Component())
   Copy Pose From Mesh() {bUseAttachedParent=true}
   n1 = "Transform (Modify) Bone - Bone: base_back_quill_c"(Alpha=1, Translation="0,5,0", Rotation="0,0,-25", Scale="0.7,1,1") {
    rotationMode=BMM_Additive
    scaleMode=BMM_Additive
    scaleSpace=BCS_BoneSpace
    translationMode=BMM_Additive
    alphaBoolBlend {blendInTime=0.1, blendOption=Linear, blendOutTime=0.12}
    boneToModify {boneName=base_back_quill_c}
    rotation {roll=-25}
    scale {x=0.7}
    translation {y=5}
   }
   n2 = "Transform (Modify) Bone - Bone: base_back_bk_quill_r"(ComponentPose=@n1, Alpha=1, Translation="0,0,0", Rotation="0,0,0", Scale="0.6,1,1") {
    rotationMode=BMM_Additive
    scaleMode=BMM_Additive
    scaleSpace=BCS_BoneSpace
    translationMode=BMM_Additive
    alphaBoolBlend {blendInTime=0.1, blendOption=Linear, blendOutTime=0.12}
    boneToModify {boneName=base_back_bk_quill_r}
    scale {x=0.6}
   }
   "Transform (Modify) Bone - Bone: base_back_bk_quill_l"(ComponentPose=@n2, Alpha=1, Translation="0,0,0", Rotation="0,0,0", Scale="0.6,1,1") {
    rotationMode=BMM_Additive
    scaleMode=BMM_Additive
    scaleSpace=BCS_BoneSpace
    translationMode=BMM_Additive
    alphaBoolBlend {blendInTime=0.1, blendOption=Linear, blendOutTime=0.12}
    boneToModify {boneName=base_back_bk_quill_l}
    scale {x=0.6}
   }
   n3 = "Transform (Modify) Bone - Bone: base_back_quill_c"(ComponentPose=Local To Component(LocalPose=Use cached pose 'Backpack'()), Alpha=1, Translation="0,0,0", Rotation="0,0,15", Scale="0.85,1,1") {
    rotationMode=BMM_Additive
    scaleMode=BMM_Additive
    scaleSpace=BCS_BoneSpace
    alphaBoolBlend {blendInTime=0.1, blendOption=Linear, blendOutTime=0.12}
    boneToModify {boneName=base_back_quill_c}
    rotation {roll=15}
    scale {x=0.85}
   }
   n4 = "Transform (Modify) Bone - Bone: base_back_bk_quill_l"(ComponentPose=@n3, Alpha=1, Translation="0,0,0", Rotation="0,0,10", Scale="0.85,1,1") {
    rotationMode=BMM_Additive
    scaleMode=BMM_Additive
    scaleSpace=BCS_BoneSpace
    alphaBoolBlend {blendInTime=0.1, blendOption=Linear, blendOutTime=0.12}
    boneToModify {boneName=base_back_bk_quill_l}
    rotation {roll=10}
    scale {x=0.85}
   }
   n5 = "Transform (Modify) Bone - Bone: base_back_bk_quill_r"(ComponentPose=@n4, Alpha=1, Translation="0,0,0", Rotation="0,0,10", Scale="0.85,1,1") {
    rotationMode=BMM_Additive
    scaleMode=BMM_Additive
    scaleSpace=BCS_BoneSpace
    alphaBoolBlend {blendInTime=0.1, blendOption=Linear, blendOutTime=0.12}
    boneToModify {boneName=base_back_bk_quill_r}
    rotation {roll=10}
    scale {x=0.85}
   }
   Output Pose(Result=Blend Poses by bool(
     BlendPose_0=Component To Local(ComponentPose=@n5)
     BlendPose_1=Use cached pose 'Backpack'()
     bActiveValue=Get is Wearing Cape() -> (is Wearing Cape:bool)
    ) {blendType=Linear})
   n6 = RigidBody!(ComponentPose=Local To Component(LocalPose=Use cached pose 'FrontHair'()), Alpha, OverrideWorldGravity:Vector="0,0,0", ExternalForce) -> (Pose) {
    bClampLinearTranslationLimitToRefPose=true
    bOverrideWorldGravity=true
    overridePhysicsAsset=&/M_MED_MixedAlley_FaceAcc_Back_Emote_Physics
    simulationSpace=BaseBoneSpace
    alphaBoolBlend {blendOption=Linear}
    baseBoneRef {boneName=neck_02}
    componentAppliedLinearAccClamp=<0 0 0>
    simSpaceSettings {maxAngularAcceleration=16, maxAngularVelocity=8, maxLinearAcceleration=700, maxLinearVelocity=700, velocityScaleZ=0.7, worldAlpha=1}
   }
   n7 = Component To Local(ComponentPose=@n6)
   n8 = RigidBody!(ComponentPose=Local To Component(LocalPose=Use cached pose 'FrontHair'()), Alpha, OverrideWorldGravity="0,0,0") -> (Pose) {
    bClampLinearTranslationLimitToRefPose=true
    bOverrideWorldGravity=true
    overridePhysicsAsset=&/M_MED_MixedAlley_FaceAcc_Back_Physics
    simulationSpace=BaseBoneSpace
    alphaBoolBlend {blendOption=Linear}
    baseBoneRef {boneName=neck_02}
    componentAppliedLinearAccClamp=<1200 1200 1200>
    simSpaceSettings {maxAngularAcceleration=16, maxAngularVelocity=8, maxLinearAcceleration=700, maxLinearVelocity=700, velocityScaleZ=0.7, worldAlpha=1}
   }
   n9 = Component To Local(ComponentPose=@n8)
   n10 = Get bIsPlayingEmote()
   n11 = "Transform (Modify) Bone - Bone: base_back_quill_c"(
    ComponentPose=Local To Component(LocalPose=Use cached pose 'Backpack_PA'())
    Alpha=1
    Translation="0,0,0"
    Rotation="0,0,-25"
    Scale="0.7,1,1"
   ) {
    rotationMode=BMM_Additive
    scaleMode=BMM_Additive
    scaleSpace=BCS_BoneSpace
    translationMode=BMM_Additive
    alphaBoolBlend {blendInTime=0.1, blendOption=Linear, blendOutTime=0.12}
    boneToModify {boneName=base_back_quill_c}
    rotation {roll=-25}
    scale {x=0.7}
   }
   n12 = "Transform (Modify) Bone - Bone: base_back_bk_quill_r"(ComponentPose=@n11, Alpha=1, Translation="0,0,0", Rotation="0,0,0", Scale="0.7,1,1") {
    rotationMode=BMM_Additive
    scaleMode=BMM_Additive
    scaleSpace=BCS_BoneSpace
    translationMode=BMM_Additive
    alphaBoolBlend {blendInTime=0.1, blendOption=Linear, blendOutTime=0.12}
    boneToModify {boneName=base_back_bk_quill_r}
    scale {x=0.7}
   }
   n13 = "Transform (Modify) Bone - Bone: base_back_bk_quill_l"(ComponentPose=@n12, Alpha=1, Translation="0,0,0", Rotation="0,0,-10", Scale="0.6,1,1") {
    rotationMode=BMM_Additive
    scaleMode=BMM_Additive
    scaleSpace=BCS_BoneSpace
    translationMode=BMM_Additive
    alphaBoolBlend {blendInTime=0.1, blendOption=Linear, blendOutTime=0.12}
    boneToModify {boneName=base_back_bk_quill_l}
    rotation {roll=-10}
    scale {x=0.6}
   }
   n14 = "Transform (Modify) Bone - Bone: dyn_back_quill_02_c"(ComponentPose=@n13, Alpha=1, Translation="0,0,0", Rotation="0,0,-20", Scale="1,1,1") {
    rotationMode=BMM_Additive
    scaleMode=BMM_Additive
    scaleSpace=BCS_BoneSpace
    translationMode=BMM_Additive
    alphaBoolBlend {blendInTime=0.1, blendOption=Linear, blendOutTime=0.12}
    boneToModify {boneName=dyn_back_quill_02_c}
    rotation {roll=-20}
   }
   n15 = "Transform (Modify) Bone - Bone: dyn_back_bk_quill_02_r"(ComponentPose=@n14, Alpha=1, Translation="0,0,0", Rotation="0,0,-30", Scale="0.75,1,1") {
    rotationMode=BMM_Additive
    scaleMode=BMM_Additive
    scaleSpace=BCS_BoneSpace
    translationMode=BMM_Additive
    alphaBoolBlend {blendInTime=0.1, blendOption=Linear, blendOutTime=0.12}
    boneToModify {boneName=dyn_back_bk_quill_02_r}
    rotation {roll=-30}
    scale {x=0.75}
   }
   n16 = "Transform (Modify) Bone - Bone: dyn_back_bk_quill_02_l"(ComponentPose=@n15, Alpha=1, Translation="0,0,0", Rotation="0,0,-45", Scale="0.75,1,1") {
    rotationMode=BMM_Additive
    scaleMode=BMM_Additive
    scaleSpace=BCS_BoneSpace
    translationMode=BMM_Additive
    alphaBoolBlend {blendInTime=0.1, blendOption=Linear, blendOutTime=0.12}
    boneToModify {boneName=dyn_back_bk_quill_02_l}
    rotation {roll=-45}
    scale {x=0.75}
   }
   n17 = Component To Local(ComponentPose=@n16)
   n18 = "Transform (Modify) Bone - Bone: base_back_quill_c"(
    ComponentPose=Local To Component(LocalPose=Use cached pose 'Backpack_PA'())
    Alpha=1
    Translation="0,0,0"
    Rotation="0,0,-25"
    Scale="0.6,1,1"
   ) {
    rotationMode=BMM_Additive
    scaleMode=BMM_Additive
    scaleSpace=BCS_BoneSpace
    translationMode=BMM_Additive
    alphaBoolBlend {blendInTime=0.1, blendOption=Linear, blendOutTime=0.12}
    boneToModify {boneName=base_back_quill_c}
    rotation {roll=-25}
    scale {x=0.6}
   }
   n19 = "Transform (Modify) Bone - Bone: base_back_bk_quill_r"(ComponentPose=@n18, Alpha=1, Translation="0,0,0", Rotation="0,0,0", Scale="0.7,1,1") {
    rotationMode=BMM_Additive
    scaleMode=BMM_Additive
    scaleSpace=BCS_BoneSpace
    translationMode=BMM_Additive
    alphaBoolBlend {blendInTime=0.1, blendOption=Linear, blendOutTime=0.12}
    boneToModify {boneName=base_back_bk_quill_r}
    scale {x=0.7}
   }
   n20 = "Transform (Modify) Bone - Bone: base_back_bk_quill_l"(ComponentPose=@n19, Alpha=1, Translation="0,0,0", Rotation="0,0,-10", Scale="0.6,1,1") {
    rotationMode=BMM_Additive
    scaleMode=BMM_Additive
    scaleSpace=BCS_BoneSpace
    translationMode=BMM_Additive
    alphaBoolBlend {blendInTime=0.1, blendOption=Linear, blendOutTime=0.12}
    boneToModify {boneName=base_back_bk_quill_l}
    rotation {roll=-10}
    scale {x=0.6}
   }
   n21 = "Transform (Modify) Bone - Bone: dyn_back_quill_02_c"(ComponentPose=@n20, Alpha=1, Translation="0,0,0", Rotation="0,0,-20", Scale="1,1,1") {
    rotationMode=BMM_Additive
    scaleMode=BMM_Additive
    scaleSpace=BCS_BoneSpace
    translationMode=BMM_Additive
    alphaBoolBlend {blendInTime=0.1, blendOption=Linear, blendOutTime=0.12}
    boneToModify {boneName=dyn_back_quill_02_c}
    rotation {roll=-20}
   }
   n22 = "Transform (Modify) Bone - Bone: dyn_back_bk_quill_02_r"(ComponentPose=@n21, Alpha=1, Translation="0,0,0", Rotation="0,0,-30", Scale="0.75,1,1") {
    rotationMode=BMM_Additive
    scaleMode=BMM_Additive
    scaleSpace=BCS_BoneSpace
    translationMode=BMM_Additive
    alphaBoolBlend {blendInTime=0.1, blendOption=Linear, blendOutTime=0.12}
    boneToModify {boneName=dyn_back_bk_quill_02_r}
    rotation {roll=-30}
    scale {x=0.75}
   }
   n23 = "Transform (Modify) Bone - Bone: dyn_back_bk_quill_02_l"(ComponentPose=@n22, Alpha=1, Translation="0,0,0", Rotation="0,0,-45", Scale="0.75,1,1") {
    rotationMode=BMM_Additive
    scaleMode=BMM_Additive
    scaleSpace=BCS_BoneSpace
    translationMode=BMM_Additive
    alphaBoolBlend {blendInTime=0.1, blendOption=Linear, blendOutTime=0.12}
    boneToModify {boneName=dyn_back_bk_quill_02_l}
    rotation {roll=-45}
    scale {x=0.75}
   }
   n24 = Component To Local(ComponentPose=@n23)
   n25 = Get MixedAlleyPack() -> (MixedAlleyPack:bool)
   Save cached pose 'Backpack'(Pose=Blend Poses by bool(
     BlendPose_0=Blend Poses by bool(BlendPose_0=@n24, BlendPose_1=@n17, bActiveValue=@n25) {blendType=Linear}
     BlendPose_1=Blend Poses by bool(BlendPose_0=@n7, BlendPose_1=@n9, bActiveValue=@n10) {blendType=Linear}
     bActiveValue=Get is Wearing Backbling() -> (is Wearing Backbling:bool)
    ) {blendType=Linear}) {cacheName=Backpack}
   n26 = RigidBody!(ComponentPose=Local To Component(LocalPose=Use cached pose 'Head Rotation'()), Alpha, OverrideWorldGravity="0,0,0", ExternalForce:Vector) -> (Pose) {
    bClampLinearTranslationLimitToRefPose=true
    bOverrideWorldGravity=true
    overridePhysicsAsset=&/M_MED_MixedAlley_FaceAcc_Front_ADS_Physics
    simulationSpace=BaseBoneSpace
    alphaBoolBlend {blendOption=Linear}
    baseBoneRef {boneName=neck_02}
    componentAppliedLinearAccClamp=<0 0 0>
    simSpaceSettings {maxAngularAcceleration=16, maxAngularVelocity=8, maxLinearAcceleration=700, maxLinearVelocity=700, velocityScaleZ=0.7, worldAlpha=1}
   }
   n27 = "Transform (Modify) Bone - Bone: base_back_quill_c"(ComponentPose=@n26, Alpha=1, Translation="0,0,0", Rotation="0,0,20", Scale="1,1,1") {
    rotationMode=BMM_Additive
    scaleMode=BMM_Additive
    scaleSpace=BCS_BoneSpace
    translationMode=BMM_Additive
    alphaBoolBlend {blendInTime=0.1, blendOption=Linear, blendOutTime=0.12}
    boneToModify {boneName=base_back_quill_c}
    rotation {roll=20}
   }
   n28 = "Transform (Modify) Bone - Bone: base_back_bk_quill_r"(ComponentPose=@n27, Alpha=1, Translation="0,5,0", Rotation="0,0,30", Scale="1,1,1") {
    rotationMode=BMM_Additive
    scaleMode=BMM_Additive
    scaleSpace=BCS_BoneSpace
    alphaBoolBlend {blendInTime=0.1, blendOption=Linear, blendOutTime=0.12}
    boneToModify {boneName=base_back_bk_quill_r}
    rotation {roll=30}
    translation {y=5}
   }
   n29 = Get scaleDownAlpha() -> (scaleDownAlpha:bool)
   n30 = "Transform (Modify) Bone - Bone: dyn_front_quill_01_r"!(ComponentPose=@n28, bAlphaBoolEnabled:bool=@n29, Translation="15,0,0", Rotation="15,0,-25", Scale="0.75,0.75,0.75") -> (Pose) {
    alphaInputType=Bool
    rotationMode=BMM_Additive
    scaleMode=BMM_Replace
    scaleSpace=BCS_BoneSpace
    alphaBoolBlend {blendInTime=0.1, blendOption=Linear, blendOutTime=0.12}
    boneToModify {boneName=dyn_front_quill_01_r}
    rotation {pitch=15, roll=-25}
    scale=<0.75 0.75 0.75>
    translation {x=15}
   }
   n31 = Component To Local(ComponentPose=@n30)
   n32 = RigidBody(ComponentPose=Local To Component(LocalPose=Use cached pose 'Head Rotation'()), OverrideWorldGravity="0,0,0") {
    bClampLinearTranslationLimitToRefPose=true
    bOverrideWorldGravity=true
    overridePhysicsAsset=&/M_MED_MixedAlley_FaceAcc_Front_Physics
    simulationSpace=BaseBoneSpace
    alphaBoolBlend {blendOption=Linear}
    baseBoneRef {boneName=neck_02}
    componentAppliedLinearAccClamp=<0 0 0>
    simSpaceSettings {maxAngularAcceleration=16, maxAngularVelocity=8, maxLinearAcceleration=700, maxLinearVelocity=700, velocityScaleZ=0.7, worldAlpha=1}
   }
   n33 = Component To Local(ComponentPose=@n32)
   n34 = RigidBody(ComponentPose=Local To Component(LocalPose=Use cached pose 'Head Rotation'()), OverrideWorldGravity="0,0,0") {
    bClampLinearTranslationLimitToRefPose=true
    bOverrideWorldGravity=true
    overridePhysicsAsset=&/M_MED_MixedAlley_FaceAcc_FrontEmote_Physics
    simulationSpace=BaseBoneSpace
    alphaBoolBlend {blendOption=Linear}
    baseBoneRef {boneName=neck_02}
    componentAppliedLinearAccClamp=<0 0 0>
    simSpaceSettings {maxAngularAcceleration=16, maxAngularVelocity=8, maxLinearAcceleration=700, maxLinearVelocity=700, velocityScaleZ=0.7, worldAlpha=1}
   }
   Save cached pose 'FrontHair'(Pose=Blend Poses by bool(
     BlendPose_0=Component To Local(ComponentPose=@n34)
     BlendPose_1=Blend Poses by bool(BlendPose_0=@n31, BlendPose_1=@n33, bActiveValue=@n29) {blendType=Linear}
     bActiveValue=Get bIsPlayingEmote()
    ) {blendType=Linear}) {cacheName=FrontHair}
   n35 = Control Rig!(
    Source=Input Pose()
    Alpha
    is Playing Emote:bool=Get bIsPlayingEmote()
    is Parachute Open:bool=Get bIsParachuteOpened()
    Is Wearing Backpack:bool=Get is Wearing Backbling()
    Is Wearing Cape:bool=Get is Wearing Cape()
   ) -> (Pose) {
    alphaBoolBlend {blendOption=Linear}
    controlRigAssetReference {blueprintRigClass=&/M_MED_MixedAlley_Pin_Head_CtrlRig.M_MED_MixedAlley_Pin_Head_CtrlRig_C}
   }
   n36 = Get scaleDownAlpha()
   n37 = "Transform (Modify) Bone - Bone: head"!(ComponentPose=Local To Component(LocalPose=@n35), bAlphaBoolEnabled:bool=@n36, Translation="0,0,0", Rotation="0,0,0", Scale="0.75,1,1") -> (Pose) {
    alphaInputType=Bool
    scaleMode=BMM_Replace
    scaleSpace=BCS_BoneSpace
    alphaBoolBlend {blendInTime=0.1, blendOption=Linear, blendOutTime=0.12}
    boneToModify {boneName=head}
    scale {x=0.75}
   }
   n38 = "Transform (Modify) Bone - Bone: neck_02"!(ComponentPose=@n37, bAlphaBoolEnabled:bool=@n36, Translation="0,0,0", Rotation="0,0,0", Scale="1,0.75,0.75") -> (Pose) {
    alphaInputType=Bool
    scaleMode=BMM_Replace
    scaleSpace=BCS_BoneSpace
    alphaBoolBlend {blendInTime=0.1, blendOption=Linear, blendOutTime=0.12}
    boneToModify {boneName=neck_02}
    scale {y=0.75, z=0.75}
   }
   n39 = "Transform (Modify) Bone - Bone: head"!(ComponentPose=@n38, Alpha:float=1, Translation="0,0,0", Rotation=Get head rot() -> (head rot:Rotator), Scale="1,1,1") -> (Pose) {rotationMode=BMM_Additive, rotationSpace=BCS_BoneSpace, alphaBoolBlend {blendOption=Linear}, boneToModify {boneName=head}}
   Save cached pose 'Head Rotation'(Pose=Component To Local(ComponentPose=@n39)) {cacheName=Head Rotation}
   n40 = RigidBody(ComponentPose=Local To Component(LocalPose=Use cached pose 'FrontHair'()), OverrideWorldGravity="0,0,0") {
    bClampLinearTranslationLimitToRefPose=true
    bOverrideWorldGravity=true
    overridePhysicsAsset=&/M_MED_MixedAlley_FaceAcc_Pack_Physics
    simulationSpace=BaseBoneSpace
    alphaBoolBlend {blendOption=Linear}
    baseBoneRef {boneName=neck_02}
    componentAppliedLinearAccClamp=<0 0 0>
    simSpaceSettings {maxAngularAcceleration=16, maxAngularVelocity=8, maxLinearAcceleration=700, maxLinearVelocity=700, velocityScaleZ=0.7, worldAlpha=1}
   }
   n41 = RigidBody(ComponentPose=Local To Component(LocalPose=Use cached pose 'FrontHair'()), OverrideWorldGravity="0,0,0") {
    bClampLinearTranslationLimitToRefPose=true
    bOverrideWorldGravity=true
    overridePhysicsAsset=&M_MED_MixedAlley_FaceAcc_Back_Emote_Physics
    simulationSpace=BaseBoneSpace
    alphaBoolBlend {blendOption=Linear}
    baseBoneRef {boneName=neck_02}
    componentAppliedLinearAccClamp=<0 0 0>
    simSpaceSettings {maxAngularAcceleration=16, maxAngularVelocity=8, maxLinearAcceleration=700, maxLinearVelocity=700, velocityScaleZ=0.7, worldAlpha=1}
   }
   Save cached pose 'Backpack_PA'(Pose=Blend Poses by bool(BlendPose_0=Component To Local(ComponentPose=@n41), BlendPose_1=Component To Local(ComponentPose=@n40), bActiveValue=Get bIsPlayingEmote()) {blendType=Linear}) {cacheName=Backpack_PA}
  ]}
 EventGraph {nodes [
   Event BlueprintUpdateAnimation()
   > n1 = Sequence!(execute) -> (then_0, then_1, then_2:Exec)
   > n2 = Branch(Condition=OR Boolean(A=Get bIsPlayingEmote(), B=Get bIsInFrontEnd()))
   n3 = GetOwnerSkeletalMeshComponent()
   n4 = BreakRotator(InRot=BreakTransform(InTransform=GetDeltaTransformFromRefPose(self=@n3, BoneName=head, BaseName=neck_02)).Rotation)
   n5 = Get Z percentage() -> (Z percentage:float)
   n6 = Get Y percentage() -> (Y percentage:float)
   n7 = Get X percentage() -> (X percentage:float)
   > n8 = Set head rot(head rot:Rotator=MakeRotator(
     Roll=float * float!(A=@n4, B=0, C:double=@n7) -> (ReturnValue)
     Pitch=float * float(A=@n4.Pitch, B=0, C=@n6)
     Yaw=float * float(A=@n4.Yaw, B=-0.3, C=@n5)
    )) -> (Output_Get:Rotator)
   n9 = "Gravity Override State Function (Parameter Asset) (Nativized)"(
    execute=[@n8.then, @n15.then]
    CharacterPartAnimInstance=Self-Reference()
    GameTimeInSeconds=GetGameTimeInSeconds()
    Parameters=/BRCosmetics/Characters/Player/Male/Medium/Heads/M_MED_SaltyBrim_Pin_Head/Meshes/CustomSkeleton/M_MED_SaltyBrim_Pin_Parameters
   )
   > Set GravityOverride(GravityOverride=@n9)
   > Set RBAN Settings Out(RBAN Settings Out=@n9.OutRigidBodyAnimNodeParameters)
   n10 = GetOwnerSkeletalMeshComponent()
   n11 = BreakRotator(InRot=BreakTransform(InTransform=GetDeltaTransformFromRefPose(self=@n10, BoneName=head, BaseName=neck_02)).Rotation)
   n12 = Get Z percentage()
   n13 = Get X percentage()
   n14 = Get Y percentage()
   n15 = Set head rot(
    execute=@n2.else
    head rot=MakeRotator(Roll=float * float(A=@n11, B=-1.5, C=@n13), Pitch=float * float(A=@n11.Pitch, B=-1.5, C=@n14), Yaw=float * float(A=@n11.Yaw, B=-1.5, C=@n12))
   )
   n16 = Set isLocalPlayer(execute=@n1.then_1, isLocalPlayer:bool=IsLocalPawn!(Actor:Actor Object Reference=GetOwningActor()) -> (IsLocalPawn:bool)) -> (Output_Get:bool)
   > n17 = Branch(Condition=AND Boolean!(A=@n16, B=Get bIsTargeting(), C:bool=Get weapon_visible() -> (weapon_visible:bool)) -> (ReturnValue))
   > Set scaleDownAlpha(scaleDownAlpha:bool=true) -> (Output_Get:bool)
   > Set Quill scale(Quill scale:double=1) -> (Output_Get:double)
   Set scaleDownAlpha(execute=@n17.else, scaleDownAlpha=false)
   > Set Quill scale(Quill scale=0)
   Set FortPawn(execute=@n1.then_2, FortPawn:Fort Pawn Object Reference=TryGetFortPawn()) -> (Output_Get:Fort Pawn Object Reference)
   n18 = Get FortPawn() -> (FortPawn:Fort Pawn Object Reference)
   > n19 = Is Valid(InputObject=@n18)
   Set weapon_visible(execute=@n19.Is Not Valid, weapon_visible:bool=false) -> (Output_Get:bool)
   n20 = Get WeaponCoreAnim!() -> (WeaponCoreAnim)
   n21 = Set notCoverCrosshair(
    execute=@n19.Is Valid
    notCoverCrosshair:bool=OR Boolean!(
     A="Equal (Enum)"(A:EFortWeaponCoreAnimation Enum=@n20, B:EFortWeaponCoreAnimation Enum=TrapPlacement)
     B="Equal (Enum)"(A=@n20, B=RocketLauncher)
     C:bool="Equal (Enum)"(A=@n20, B=GoingCommando)
    ) -> (ReturnValue)
   ) -> (Output_Get:bool)
   > n22 = Branch(Condition=@n21)
   > Set weapon_visible(weapon_visible=false)
   n23 = IsWeaponHolstered!(self:Fort Pawn Object Reference=@n18) -> (ReturnValue:bool)
   n24 = Get CurrentWeapon() -> (CurrentWeapon:Fort Weapon Object Reference)
   Set weapon_visible(execute=@n22.else, weapon_visible=AND Boolean!(A=IsValid(Object=@n24), B=NOT Boolean(A=@n23)) -> (ReturnValue))
   Event OnCharacterCustomizationLoaded()
   > Set Backpack Mesh(Backpack Mesh:Skeletal Mesh Component Object Reference=Get Array Item with EFort Custom Part Type(PartType=Backpack)) -> (Output_Get:Skeletal Mesh Component Object Reference)
   n25 = Get Backpack Mesh() -> (Backpack Mesh:Skeletal Mesh Component Object Reference)
   > Set MixedAlleyPack(MixedAlleyPack:bool=MatchesWildcard(SourceString=GetDisplayName(Object=@n25), Wildcard=*MixedAlley*)) -> (Output_Get:bool)
   > Set Dynamic_Alpha(Dynamic_Alpha:float=1) -> (Output_Get:float)
   n26 = CheckForBackbling()
   > Set is Wearing Backbling(is Wearing Backbling:bool=AND Boolean!(A=@n26, B=NOT Boolean(A=@n26.bIsBackblingACape), C:bool=NOT Boolean(A=@n26.bIsBackblingLong)) -> (ReturnValue)) -> (Output_Get:bool)
   > Set is Wearing Cape(is Wearing Cape:bool=OR Boolean!(A=@n26.bIsBackblingACape, B=@n26.bIsBackblingLong) -> (ReturnValue)) -> (Output_Get:bool)
   > Set is Wearing Backbling or Cape(is Wearing Backbling or Cape:bool=OR Boolean(A=Get is Wearing Backbling(), B=Get is Wearing Cape())) -> (Output_Get:bool)
   Event OnBeginSkydiving()
   > Set Dynamic_Alpha(Dynamic_Alpha=0.5)
   Event OnEndSkydiving()
   > Set Dynamic_Alpha(Dynamic_Alpha=1)
   TryGetPawnOwner()
  ]}
]
variables [
 Backpack Mesh {type=SkeletalMeshComponent}
 Dynamic_Alpha {type=number, default=0.4}
 FortPawn {type=FortPawn, default=null}
 GravityOverride {type=Vector}
 head rot {type=Rotator}
 is Wearing Backbling {type=boolean}
 is Wearing Backbling or Cape {type=boolean}
 is Wearing Cape {type=boolean}
 isLocalPlayer {type=boolean}
 MixedAlleyPack {type=boolean}
 notCoverCrosshair {type=boolean}
 Quill scale {type=number}
 RBAN Settings Out {type=FortRigidBodyAnimNodeParameters}
 scaleDownAlpha {type=boolean}
 weapon_visible {type=boolean}
 X percentage {type=number, default=0.4}
 Y percentage {type=number, default=0.4}
 Z percentage {type=number, default=0.8}
]
